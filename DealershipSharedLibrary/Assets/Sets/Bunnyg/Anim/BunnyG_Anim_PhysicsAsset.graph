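# Physics asset for the BunnyG animated character, defining the collision and joint setup of a two-legged skeleton with ears.
# Single capsules cover the pelvis, two spine bones, head, one bone per ear, and each arm and leg down to palm and ankle.
# Constraints form a tree from the pelvis through spine to head, ears and limbs.
# Every joint shares identical swing and twist limit angles, with no per-joint tuning such as hinge-like knees or elbows.
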PhysicsAsset /DealershipSharedLibrary/Assets/Sets/Bunnyg/Anim/BunnyG_Anim_PhysicsAsset
bodies {
 L_ankle_JNT [{shapeType=Capsule, length=14.4212, radius=10.1012, center=<5.9144 -0.4 1.8416>, rotation=<rot 0 -127.7418 -84.945>}]
 L_ear_02_JNT [{shapeType=Capsule, length=20.6935, radius=8.8862, center=<18.0576 -1.0466 -2.0612>, rotation=<rot 0 -103.237 -70.6372>}]
 L_elbow_JNT [{shapeType=Capsule, length=21.7063, radius=8.1062, center=<19.4032 0.1705 0.3561>, rotation=<rot 0 -87.6976 -89.1409>}]
 L_knee_JNT [{shapeType=Capsule, length=34.0524, radius=13.6105, center=<25.9973 2.5491 1.0435>, rotation=<rot 0 87.9474 -88.3195>}]
 L_palm_JNT [{shapeType=Capsule, length=14.3774, radius=12.5019, center=<4.0663 1.9289 1.3366>, rotation=<rot 0 165.9385 -73.5083>}]
 L_shoulder_JNT [{shapeType=Capsule, length=19.312, radius=8.9823, center=<10.2915 -1.9711 1.7532>, rotation=<rot 0 -92.673 -88.9018>}]
 L_thigh_JNT [{shapeType=Capsule, length=32.6214, radius=18.5161, center=<20.3392 0.4762 -0.2417>, rotation=<rot 0 78.1084 -84.5439>}]
 R_ankle_JNT [{shapeType=Capsule, length=14.2677, radius=10.1015, center=<5.6202 -0.9171 -1.8303>, rotation=<rot 0 51.0025 -84.8562>}]
 R_ear_02_JNT [{shapeType=Capsule, length=20.6935, radius=8.8862, center=<18.0577 -1.0467 2.0612>, rotation=<rot 0 76.763 -70.6372>}]
 R_elbow_JNT [{shapeType=Capsule, length=20.6358, radius=8.0945, center=<20.4448 0.0694 -0.3412>, rotation=<rot 0 93.0637 -89.2304>}]
 R_knee_JNT [{shapeType=Capsule, length=31.8082, radius=12.9902, center=<28.2249 2.7981 -1.071>, rotation=<rot 0 -93.7489 -87.9398>}]
 R_palm_JNT [{shapeType=Capsule, length=14.3773, radius=12.5037, center=<4.0521 1.9311 -1.3156>, rotation=<rot 0 -14.0223 -73.7612>}]
 R_shoulder_JNT [{shapeType=Capsule, length=17.2635, radius=8.6612, center=<11.8354 -1.9968 -1.3302>, rotation=<rot 0 83.0334 -83.4881>}]
 R_thigh_JNT [{shapeType=Capsule, length=34.524, radius=18.4221, center=<24.4304 -0.2793 0.3987>, rotation=<rot 0 -104.1407 -88.9517>}]
 head_JNT [{shapeType=Capsule, length=27.297, radius=25.9145, center=<20.1487 -2.5126 -0.062>, rotation=<rot 0.1735 -90 0.3506>}]
 pelvis_JNT [{shapeType=Capsule, length=23.2058, radius=22.1138, center=<0.457 7.7481 0.0072>, rotation=<rot 0 173.4964 -75.1781>}]
 spineA_JNT [{shapeType=Capsule, length=15.464, radius=15.2801, center=<13.2759 -3.6508 0.4662>, rotation=<rot -7.5321 -90 -2.8169>}]
 spineB_JNT [{shapeType=Capsule, length=19.2485, radius=17.0273, center=<3.9739 -0.0959 -1.9115>, rotation=<rot 0 -63.5387 -60.7869>}]
}
constraints [
 {bone1Name=spineA_JNT, bone2Name=pelvis_JNT, swing1LimitDegrees=45, swing2LimitDegrees=45, twistLimitDegrees=45}
 {bone1Name=spineB_JNT, bone2Name=spineA_JNT, swing1LimitDegrees=45, swing2LimitDegrees=45, twistLimitDegrees=45}
 {bone1Name=head_JNT, bone2Name=spineB_JNT, swing1LimitDegrees=45, swing2LimitDegrees=45, twistLimitDegrees=45}
 {bone1Name=L_ear_02_JNT, bone2Name=head_JNT, swing1LimitDegrees=45, swing2LimitDegrees=45, twistLimitDegrees=45}
 {bone1Name=R_ear_02_JNT, bone2Name=head_JNT, swing1LimitDegrees=45, swing2LimitDegrees=45, twistLimitDegrees=45}
 {bone1Name=L_shoulder_JNT, bone2Name=spineB_JNT, swing1LimitDegrees=45, swing2LimitDegrees=45, twistLimitDegrees=45}
 {bone1Name=L_elbow_JNT, bone2Name=L_shoulder_JNT, swing1LimitDegrees=45, swing2LimitDegrees=45, twistLimitDegrees=45}
 {bone1Name=L_palm_JNT, bone2Name=L_elbow_JNT, swing1LimitDegrees=45, swing2LimitDegrees=45, twistLimitDegrees=45}
 {bone1Name=R_shoulder_JNT, bone2Name=spineB_JNT, swing1LimitDegrees=45, swing2LimitDegrees=45, twistLimitDegrees=45}
 {bone1Name=R_elbow_JNT, bone2Name=R_shoulder_JNT, swing1LimitDegrees=45, swing2LimitDegrees=45, twistLimitDegrees=45}
 {bone1Name=R_palm_JNT, bone2Name=R_elbow_JNT, swing1LimitDegrees=45, swing2LimitDegrees=45, twistLimitDegrees=45}
 {bone1Name=L_thigh_JNT, bone2Name=pelvis_JNT, swing1LimitDegrees=45, swing2LimitDegrees=45, twistLimitDegrees=45}
 {bone1Name=L_knee_JNT, bone2Name=L_thigh_JNT, swing1LimitDegrees=45, swing2LimitDegrees=45, twistLimitDegrees=45}
 {bone1Name=L_ankle_JNT, bone2Name=L_knee_JNT, swing1LimitDegrees=45, swing2LimitDegrees=45, twistLimitDegrees=45}
 {bone1Name=R_thigh_JNT, bone2Name=pelvis_JNT, swing1LimitDegrees=45, swing2LimitDegrees=45, twistLimitDegrees=45}
 {bone1Name=R_knee_JNT, bone2Name=R_thigh_JNT, swing1LimitDegrees=45, swing2LimitDegrees=45, twistLimitDegrees=45}
 {bone1Name=R_ankle_JNT, bone2Name=R_knee_JNT, swing1LimitDegrees=45, swing2LimitDegrees=45, twistLimitDegrees=45}
]
