Material /DealershipSharedLibrary/Assets/Sets/Bunnyg/Materials/MM_DLR_FakeParticles_PointDot_Noise_Text_TEMP
graph {
 properties {
  bUsedWithNanite=true
  bUsedWithNiagaraMeshParticles=true
  blendMode=BLEND_Masked
  pixelDepthOffsetMode=PDOM_Legacy
  shadingModel=MSM_Unlit
  twoSided=true
  usedShadingModels=Unlit
 }
 nodes [
  Add(A=NamedRerouteUsage(), B=NamedRerouteUsage())
  n1 = NamedRerouteUsage()
  n2 = VertexColor()
  n3 = Constant() {r=0.55}
  n4 = Saturate(None=Ceil(None=Subtract(A=@n2.A, B=@n3))) {desc=LETTERS}
  n5 = Multiply(A=VectorParameter() {parameterName=TextColor, defaultValue=<rgba 1 0.75 0.035 1>}.RGB, B=@n4)
  n6 = OneMinus(None=@n2.A)
  n7 = Saturate(None=Ceil(None=Subtract(A=@n6, B=@n3))) {desc=OUTER}
  n8 = Multiply(A=OneMinus(None=Add(A=@n4, B=@n7)) {desc=MIDDLE}, B=VectorParameter() {parameterName=MidColor, defaultValue=<rgba 1 0.035 0.015 1>}.RGB)
  n9 = VectorParameter() {parameterName=BackgroundColor, defaultValue=<rgba 1 1 1 1>}
  n10 = Multiply(A=Add(A=Add(A=@n5, B=@n8), B=Multiply(A=@n7, B=@n9.RGB)), B=@n1)
  n11 = StaticSwitchParameter(True=@n1, False=VectorParameter() {parameterName=DissolveColor, defaultValue=<rgba 5 1 0.2 1>}.RGB) {parameterName=UseVerColor}
  n12 = AppendVector(A=TextureCoordinate() {coordinateIndex=1}, B=ComponentMask(None=TextureCoordinate() {coordinateIndex=2}) {r=true})
  n13 = Subtract(A=Distance(A=Transform(None=@n12) {transformSourceType=TRANSFORMSOURCE_Local}, B=NamedRerouteUsage()), B=NamedRerouteUsage())
  n14 = TextureSampleParameter2D(UVs=Multiply(A=@n13, B=ScalarParameter() {defaultValue=0.01})) {parameterName=vertWave, samplerSource=SSM_Clamp_WorldGroupSettings, texture=/Game/Effects/Fort_Effects/Textures/Ramps/T_Ramp_up}
  n15 = Multiply(A=@n14.R, B=NamedRerouteUsage())
  n16 = OneMinus(None=Add(A=@n13, B=ScalarParameter() {defaultValue=-90, parameterName=ColorMaskAdd}) {constB=-90})
  MaterialFunctionCall(Emissive=LinearInterpolate(A=StaticSwitchParameter(True=@n1, False=@n10) {parameterName=UseVerColor}, B=Multiply(A=@n11, B=@n15), Alpha=Saturate(None=@n16))) {materialFunction=&MF_FX_EmissiveAdjustments}
  n17 = OneMinus(None=Multiply(A=@n13, B=Constant() {r=0.01}))
  NamedRerouteDeclaration(None=Multiply(A=Saturate(None=@n17), B=Constant() {r=128})) {name=NoiseMask, nodeColor=<rgba 0.526 0.1427 0 1>}
  n18 = Time()
  n19 = Multiply(A=@n18, B=ScalarParameter() {defaultValue=0.4, parameterName=Wavy Speed})
  n20 = ComponentMask(None=TextureCoordinate() {coordinateIndex=3}) {r=true}
  n21 = Add(A=@n19, B=@n20)
  n22 = Sine(None=@n21)
  n23 = Cosine(None=@n21) {period=0.75}
  n24 = Add(A=OneMinus(None=@n19), B=@n20)
  n25 = ConstantBiasScale(None=AppendVector(A=AppendVector(A=@n22, B=@n23), B=Sine(None=@n24) {period=1.5})) {bias=-0.5}
  n26 = Constant() {r=64}
  NamedRerouteDeclaration(None=StaticSwitch(True=Multiply(A=@n25, B=@n26), False=Constant(), Value=StaticBoolParameter() {defaultValue=true, parameterName=Wavy})) {name=Wavy, nodeColor=<rgba 0 0.9907 1 1>}
  n27 = ScalarParameter() {defaultValue=0.5, parameterName=SpeedNoiseVar}
  n28 = Multiply(A=@n18, B=Constant() {r=0.1})
  n29 = MaterialFunctionCall!(
   NoiseAmount=Add(A=NamedRerouteUsage(), B=NamedRerouteUsage())
   CircleMaskAlpha=StaticBool()
   "Lifetime(Alpha)"=ComponentMask(None=NamedRerouteUsage()) {r=true}
   Scale=ComponentMask(None=NamedRerouteUsage()) {r=true}
   Mesh Scale=ScalarParameter() {defaultValue=4, parameterName=MeshScale}
   NoiseScale=ScalarParameter() {defaultValue=2, parameterName=NoiseScale}
   NoiseMovementSpeed=ComponentMask(None=Add(A=@n27, B=@n28)) {r=true}
  ) {materialFunction=&/MF_PointDot_Noise_TEMP_N}
  NamedRerouteDeclaration(None=@n29.WorldPosition) {name=WPO, nodeColor=<rgba 0.0104 0.0831 1 1>}
  MaterialExpressionMeshEmitterVertexColor_2 = ParticleColor()
  NamedRerouteDeclaration(None=@MaterialExpressionMeshEmitterVertexColor_2.A) {name=ParticleAlpha, nodeColor=<rgba 0.9844 0.1743 0.1793 1>}
  NamedRerouteDeclaration(None=@MaterialExpressionMeshEmitterVertexColor_2.RGB) {name=ParticleRGB, nodeColor=<rgba 0.4263 0.2488 0.9844 1>}
  MaterialExpressionMeshEmitterDynamicParameter_2 = DynamicParameter() {defaultValue=<rgba 115.114 3.412 0.316 0>, paramNames [NoiseAmount, NoiseBrightness, AddedNoise, Remove]}
  NamedRerouteDeclaration(None=@MaterialExpressionMeshEmitterDynamicParameter_2.NoiseAmount) {name=DP NoiseAmount, nodeColor=<rgba 1 0 0.2812 1>}
  NamedRerouteDeclaration(None=@MaterialExpressionMeshEmitterDynamicParameter_2.NoiseBrightness) {name=DP NoiseBrightness, nodeColor=<rgba 1 0 0.2812 1>}
  NamedRerouteDeclaration(None=@MaterialExpressionMeshEmitterDynamicParameter_2.AddedNoise) {name=DP AddedNoise, nodeColor=<rgba 1 0 0.2812 1>}
  NamedRerouteDeclaration(None=@MaterialExpressionMeshEmitterDynamicParameter_2.Remove) {name=DP Remove, nodeColor=<rgba 1 0 0.2812 1>}
  NamedRerouteDeclaration(None=@n29.Alpha) {name=Alpha, nodeColor=<rgba 0.5938 0.6112 1 1>}
  NamedRerouteUsage()
 ]
}
